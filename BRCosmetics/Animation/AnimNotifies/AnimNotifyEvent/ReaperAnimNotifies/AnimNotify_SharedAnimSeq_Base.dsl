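; Base anim notify for animation sequences shared across skeletal meshes, spawning a particle effect chosen by the mesh that plays the animation.
; Received_Notify looks up the component's skeletal mesh in the Selection Map variable.
; A valid mapped effect spawns attached to that component at the Attach Point Name socket, and the notify returns false.
; No effect spawns when the mesh is unmapped or its entry is invalid.
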(fn Received_Notify
  (params (MeshComp :type Skeletal%20Mesh%20Component%20Object%20Reference) (Animation :type Anim%20Sequence%20Base%20Object%20Reference) (EventReference :type Anim%20Notify%20Event%20Reference%20Structure%20%28by%20ref%29))
  (bind _skeletalmesh (Variables|/Script/Engine.SkinnedMeshComponent|GetSkeletalMesh MeshComp))
  (bind _selection_map (Variables|Default|GetSelection%20Map))
  (bind (_value _returnvalue) (Utilities|Map|Find _selection_map _skeletalmesh))
  (if _returnvalue
    (Utilities|IsValid _value
      (:Is%20Valid
        (Effects|SpawnEmitterAttached _value MeshComp (Variables|Default|GetAttach%20Point%20Name))
        (return false)))))
